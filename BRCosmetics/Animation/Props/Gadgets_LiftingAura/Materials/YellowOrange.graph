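# Masked lit material for the Lifting Aura emote gadget prop, usable on skeletal meshes, with a dithered fade near the top of the mesh.
# Pre-skinned position, normalized to local bounds in a vertex interpolator, gives an inverted height gradient that stays fixed through animation.
# Min/Max remap and Power parameters shape that gradient before DitherTemporalAA simulates soft transparency.
# Diffuse, normal and mask textures supply the surface maps.
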Material /BRCosmetics/Animation/Props/Gadgets_LiftingAura/Materials/YellowOrange
graph {
 properties {
  bUsedWithMeshDeformer=true
  bUsedWithMorphTargets=true
  bUsedWithSkeletalMesh=true
  blendMode=BLEND_Masked
  pixelDepthOffsetMode=PDOM_Legacy
  usedShadingModels=Default Lit
 }
 nodes [
  n1 = PreSkinnedLocalBounds()
  n2 = Divide(A=Subtract(A=PreSkinnedPosition(), B=@n1.Min), B=@n1.Extents)
  n3 = OneMinus(None=ComponentMask(None=VertexInterpolator(VS=@n2).PS) {b=true})
  n4 = ScalarParameter() {defaultValue=0.025, parameterName=Min}
  n5 = ScalarParameter() {defaultValue=0.1, parameterName=Max}
  MaterialFunctionCall(Alpha Threshold=Power(
    Base=MaterialFunctionCall(Input=@n3, Input Low=@n4, Input High=@n5) {materialFunction=&RemapValue_0-1}.Result
    Exp=ScalarParameter() {defaultValue=3, parameterName=Power}
   )) {materialFunction=&DitherTemporalAA}
  TextureSample() {texture=/BRCosmetics/Animation/Props/Gadgets_LiftingAura/Textures/T_LiftingAura_Emote_Prop_D}
  TextureSample() {samplerType=SAMPLERTYPE_Normal, texture=/BRCosmetics/Animation/Props/Gadgets_LiftingAura/Textures/T_LiftingAura_Emote_Prop_N}
  TextureSample() {samplerType=SAMPLERTYPE_Masks, texture=/BRCosmetics/Animation/Props/Gadgets_LiftingAura/Textures/T_LiftingAura_Emote_Prop_S}
 ]
}
